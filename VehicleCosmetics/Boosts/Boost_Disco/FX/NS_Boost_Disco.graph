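# Niagara system for the Disco vehicle boost cosmetic, placed at the car's boost position.
# BOLADISCO spins and wobbles a mesh that scales in, while RIBBON and dots form the trail and Exp and rays add burst and glow sprites.
# Exp and rays merge two sprite behaviors each through spawn groups.
# User parameters switch default multicolor or painted variant looks, with frontend preview adjusting trail and dot motion.
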NiagaraSystem /VehicleCosmetics/Boosts/Boost_Disco/FX/NS_Boost_Disco
dependencies {
 usedDataInterfaces [
  &NiagaraDataInterfaceCurve
  &NiagaraDataInterfaceArrayColor
  &NiagaraDataInterfaceArrayFloat
  &NiagaraDataInterfaceSkeletalMesh
 ]
 usedDynamicInputs []
 usedModules [
  &self:SystemSpawnScript.SystemScriptSource.SystemScriptGraph.NiagaraNodeAssignment_0.SetVariables_#1
  &SystemState
 ]
}
emitters {
 BOLADISCO {
  emitterUpdateScript [
   &EmitterState(
    Inactive Response=::NewEnumerator1
    Life Cycle Mode=::NewEnumerator1
    Loop Behavior=::NewEnumerator1
    Loop Delay=0.2
    Loop Duration Mode=::NewEnumerator1
    UseLoopDelay=false
   )
   &SpawnBurst_Instantaneous(Use Spawn Probability=false, UseLoopCountLimit=false)
   &SpawnRate(SpawnRate=50, Use Spawn Probability=false) {enabled=false}
  ]
  particleSpawnScript [
   &VC_PaintedVariantsOverride(
    Black=<rgba 0.0729 0.0729 0.0725 1>
    BurntSienna=<rgba 0.3 0.0624 0.012 1>
    Cobalt=<rgba 0.1407 0.2715 1.4219 1>
    Crimson=<rgba 1.4453 0.0716 0.113 1>
    ForestGreen=<rgba 0.2475 1.4141 0.1414 1>
    Grey=<rgba 0.2344 0.2344 0.2344 1>
    Lime=<rgba 0.8068 1.0703 0.0891 1>
    Orange=<rgba 1 0.1667 0 1>
    Pink=<rgba 1.075 0.1792 0.8111 1>
    Purple=<rgba 0.3787 0 1.3 1>
    Saffron=<rgba 1.0156 0.7959 0.0741 1>
    SkyBlue=<rgba 0.1438 1.0046 1.4375 1>
    TitaniumWhite=<rgba 1 1 1 1>
   )
   &InitializeParticle(
    Color Mode=::NewEnumerator0
    Lifetime=$Emitter.CurrentLoopDuration
    Mesh Scale=<0.7 0.7 0.7>
    Mesh Scale Mode=::NewEnumerator4
    Position Offset=&AddVectorToPosition(
     Position=<-70 0 0>
     Vector=&AddVectorToPosition(Position=$User.DTCarOffset, Vector=Linked() {linkedVariable {name=User.SingleBoostPos, type=NiagaraPosition}})
    )
    Position Offset Coordinate Space=::Local
    UsePositionOffset=true
    Write Material Random=true
   )
   &AddVelocity(Velocity=<-1500 0 0>) {enabled=false}
   &Color(Color=&MakeCustomLinearColorFromBool(Bool=$User.bIsPainted, True Color=$Particles.Color, False Color=<rgba 1.2 1.2 1.2 1>))
  ]
  particleUpdateScript [
   &ParticleState(Kill Particles When Lifetime Has Elapsed=false)
   &ShapeLocation(
    MaskSpawnGroup=false
    Offset=&MakeVector(
     X=&Sine(Normalized Angle=$Emitter.Age, Period=0.5, Scale=1, Bias=0)
     Y=&Sine(Normalized Angle=$Emitter.Age, Period=0.65, Scale=1, Bias=0.3)
     Z=&Sine(Normalized Angle=$Emitter.Age, Period=0.8, Scale=1, Bias=0.6)
    )
    Offset Mode=::NewEnumerator0
    Override Seed=false
    Shape Origin=$Particles.Initial.Position
    Sphere Radius=0
    UseOwnerScale=true
   )
   &ScaleMeshSize(Scale Factor=&VectorFromFloat(Value=&FloatFromCurve(
      FloatCurve=DataInterface() {propertyValues {Curve {keys [{tangentMode=RCTM_User, time=0, value=0}, {tangentMode=RCTM_User, time=0.25, value=1}]}}}
      CurveIndex=$Emitter.Age
     ))) {module=ScaleMeshSize001}
   &ScaleMeshSize(Scale Factor=&VectorFromFloat(Value=&FloatFromCurve(FloatCurve=DataInterface() {propertyValues {Curve {keys [<at 0 0>,<at 0.2 1>]}}}))) {enabled=false}
   &UpdateMeshOrientation(Rotation Vector=<0 0 0.33>)
   &SolveForcesAndVelocity(Clamp Velocity=false, Limit Acceleration=false)
   &DynamicMaterialParameters(
    Index 0 Param 1=&MakeCustomFloatFromBool(Bool=$User.bIsPainted, True Float=$User.Fresnel, False Float=5)
    Index 0 Param 2=&MakeCustomFloatFromBool(Bool=$User.bIsPainted, True Float=&RandomRangeFloat(Minimum=2, Maximum=1.5), False Float=2.0386)
    Index 0 Param 3=&RandomRangeFloat(Minimum=0.8, Maximum=1)
    Index 0 Param 4=&MakeCustomFloatFromBool(Bool=$User.bIsPainted, True Float=0, False Float=0.3)
    Index 1 Param 1=&MakeCustomFloatFromBool(Bool=$User.bIsPainted, True Float=Linked() {linkedVariable {name=User.Gap_Metallic/Spec, type=NiagaraFloat}}, False Float=1)
    Index 1 Param 2=&MakeCustomFloatFromBool(Bool=$User.bIsPainted, True Float=Linked() {linkedVariable {name=User.Square_Metallic/Specular, type=NiagaraFloat}}, False Float=1)
    Index 1 Param 3=&MakeCustomFloatFromBool(Bool=$User.bIsPainted, True Float=$User.Gap_Roughness, False Float=0.75)
    Index 1 Param 4=&MakeCustomFloatFromBool(Bool=$User.bIsPainted, True Float=$User.Square_Roughness, False Float=2)
    Module.Index0Float0Write=true
    Module.Index0Float1Write=true
    Module.Index0Float2Write=true
    Module.Index0Float3Write=true
    Module.Index1Float0Write=true
    Module.Index1Float1Write=true
    Module.Index1Float2Write=true
    Module.Index1Float3Write=true
    Write Parameter Index 1=::NewEnumerator1
   )
  ]
  rendererClasses [&NiagaraSpriteRendererProperties, &NiagaraMeshRendererProperties]
  renderers [
   {rendererClass=&NiagaraSpriteRendererProperties}
   {rendererClass=&NiagaraMeshRendererProperties, rendererIndex=1}
  ]
 }
 Exp {
  emitterUpdateScript [
   &EmitterState(Life Cycle Mode=::NewEnumerator1, Loop Behavior=::NewEnumerator1, Loop Duration=1, UseLoopDelay=false)
   &SpawnBurst_Instantaneous(
    Spawn Count=&RandomRangeInt(Minimum=50, Maximum=100)
    Spawn Group=1
    Use Spawn Probability=false
    UseLoopCountLimit=false
   )
   &SpawnRate(SpawnGroup=1, SpawnRate=300, Use Spawn Probability=false) {enabled=false, module=SpawnRate001}
   &SpawnRate(SpawnRate=20, Use Spawn Probability=false)
   &VC_FilteredSocketCount()
  ]
  particleSpawnScript [
   &VC_PaintedVariantsOverride(
    Black=<rgba 0.3 0.3 0.3 2>
    BurntSienna=<rgba 3 1.2 0.3 2>
    Cobalt=<rgba 0 0.1667 10 1>
    Crimson=<rgba 4 0.3333 0 1>
    DefaultColor=&SelectLinearColorFromArray(
     Array Sampling Mode=::NewEnumerator0
     Direct Array Index=?
     Direct Array Mode=::NewEnumerator1
     Color Selection Array=DataInterface() {propertyValues {ColorData [<rgba 1 0.5 0.1 3>, <rgba 0.1 1 0.1 3>, <rgba 0.1 0.1 1 3>, <rgba 1 0.1 0.1 3>, <rgba 0 0 0 0>]}}
     Random Seed=0
    )
    ForestGreen=<rgba 0 1 0.0505 1>
    Grey=<rgba 0.7 0.7 0.7 1>
    Lime=<rgba 0.75 1.5 0 1>
    Orange=<rgba 2 0.5 0 1>
    Pink=<rgba 2 0.2396 1.3598 1>
    Purple=<rgba 0.1667 0 2 1>
    R1 BurntSienna:LinearColor=<rgba 1 0.2904 0.1 1>
    R1 Cobalt:LinearColor=<rgba 0 0.0833 5 1>
    R1 Crimson:LinearColor=<rgba 3 0 0.1 1>
    R1 ForestGreen:LinearColor=<rgba 0 2 0.1011 1>
    R1 Grey:LinearColor=<rgba 0.4 0.4 0.4 1>
    R1 Lime:LinearColor=<rgba 1.1 2 0.2 1>
    R1 Orange:LinearColor=<rgba 1 0.3488 0 1>
    R1 Pink:LinearColor=<rgba 2 0.2396 1.3598 1>
    R1 Purple:LinearColor=<rgba 0.1667 0 2 1>
    R1 Saffron:LinearColor=<rgba 2 1.58 0.2 1>
    R1 SkyBlue:LinearColor=<rgba 0 1.3308 2 1>
    R1 TitaniumWhite:LinearColor=<rgba 2 2 2 1>
    Saffron=<rgba 1 0.7667 0 1.5>
    SkyBlue=<rgba 0.5 3.4942 5 0.8>
    TitaniumWhite=<rgba 2 2 2 1>
    Use R1 Painted Overrides=true
   )
   /SetVariables(Particles.vistag:NiagaraInt32=&SpawnGroup()) {isSetParameters=true, module=SetVariables_#2}
   &InitializeParticle(
    Color Mode=::NewEnumerator0
    Lifetime=/VehicleCosmetics/FX/Niagara/Modules/MergeEmitters/VC_FloatBySpawnGroup(
     SpawnGroup or ExecIndex:NiagaraInt32=0
     CustomInt:NiagaraInt32=0
     MaximumSpawnGroups:NiagaraInt32=2
     Float SG 0:NiagaraFloat=1
     Float SG 1:NiagaraFloat=&RandomRangeFloat(Minimum=0.5, Maximum=0.3)
     Float SG 2:NiagaraFloat=0
     Float SG 3:NiagaraFloat=0
     Float SG 4:NiagaraFloat=0
     Float SG 5:NiagaraFloat=0
    )
    Position Offset=&AddVectorToPosition(
     Position=<-70 0 0>
     Vector=&AddVectorToPosition(Position=$User.DTCarOffset, Vector=Linked() {linkedVariable {name=User.SingleBoostPos, type=NiagaraPosition}})
    )
    Position Offset Coordinate Space=::Local
    Sprite Rotation Mode=::NewEnumerator2
    Sprite Size Mode=::NewEnumerator3
    Uniform Sprite Size=&VC_FloatBySpawnGroup(
     SpawnGroup or ExecIndex:NiagaraInt32=0
     CustomInt:NiagaraInt32=0
     MaximumSpawnGroups:NiagaraInt32=2
     Float SG 0:NiagaraFloat=11.6
     Float SG 1:NiagaraFloat=/Niagara/DynamicInputs/TypeConversions/Vector2DLength(Vector:Vector2f=<15 25>)
     Float SG 2:NiagaraFloat=0
     Float SG 3:NiagaraFloat=0
     Float SG 4:NiagaraFloat=0
     Float SG 5:NiagaraFloat=0
    )
    UsePositionOffset=true
    Write Material Random=true
   )
   &ShapeLocation(MaskSpawnGroup=true, Override Seed=false, Spawn Group Mask=1, Sphere Radius=40, UseOwnerScale=true)
   &AddVelocity(Velocity=&RandomRangeVector(Minimum=<-200 -200 -200>, Maximum=<200 200 200>))
   &Color(Color=/VehicleCosmetics/FX/Niagara/Modules/MergeEmitters/VC_Dynamic_ColorBySpawnGroup(
     Color SG 0:LinearColor=&MakeCustomLinearColorFromBool(
      Bool=$User.bIsPainted
      True Color=&Multiply_LinearColorByFloat(Linear Color=$User.PaintedColor_2, Float=2)
      False Color=&SelectLinearColorFromArray(
       Array Sampling Mode=::NewEnumerator0
       Direct Array Index=?
       Direct Array Mode=::NewEnumerator1
       Color Selection Array=DataInterface() {propertyValues {ColorData [<rgba 1 0.5 0.1 5>, <rgba 0.1 1 0.1 3>, <rgba 0.1 0.1 1 5>, <rgba 1 0.1 0.1 3>, <rgba 0.1 0.1 1 5>]}}
       Random Seed=0
      )
     )
     MaximumSpawnGroups:NiagaraInt32=2
     Color SG 1:LinearColor=&MakeCustomLinearColorFromBool(Bool=$User.bIsPainted, True Color=$User.PaintedColor_3, False Color=<rgba 1 1 1 1>)
     Color SG 2:LinearColor=<rgba 1 1 1 1>
     Color SG 3:LinearColor=<rgba 1 1 1 1>
     Color SG 4:LinearColor=<rgba 1 1 1 1>
    ))
  ]
  particleUpdateScript [
   &ParticleState()
   &ScaleSpriteSize(
    Scale Sprite Size Mode=::NewEnumerator0
    Uniform Scale Factor=&VC_FloatBySpawnGroup(
     SpawnGroup or ExecIndex:NiagaraInt32=0
     CustomInt:NiagaraInt32=0
     MaximumSpawnGroups:NiagaraInt32=2
     Float SG 0:NiagaraFloat=20
     Float SG 1:NiagaraFloat=&FloatFromCurve(FloatCurve=DataInterface() {propertyValues {Curve {keys [{tangentMode=RCTM_User, time=0, value=0}, {tangentMode=RCTM_User, time=0.5, value=1}]}}})
     Float SG 2:NiagaraFloat=0
     Float SG 3:NiagaraFloat=0
     Float SG 4:NiagaraFloat=0
     Float SG 5:NiagaraFloat=0
    )
   )
   &SolveForcesAndVelocity(Clamp Velocity=false, Limit Acceleration=false)
   &ScaleColor(
    Scale Alpha=&VC_FloatBySpawnGroup(
     SpawnGroup or ExecIndex:NiagaraInt32=0
     CustomInt:NiagaraInt32=0
     MaximumSpawnGroups:NiagaraInt32=2
     Float SG 0:NiagaraFloat=2
     Float SG 1:NiagaraFloat=&FloatFromCurve(FloatCurve=DataInterface() {propertyValues {Curve {keys [<at 0 1>,<at 1 0>]}}})
     Float SG 2:NiagaraFloat=0
     Float SG 3:NiagaraFloat=0
     Float SG 4:NiagaraFloat=0
     Float SG 5:NiagaraFloat=0
    )
    Scale RGB=/VehicleCosmetics/FX/Niagara/Modules/MergeEmitters/VC_VectorBySpawnGroup(
     SpawnGroup or ExecIndex:NiagaraInt32=0
     CustomInt:NiagaraInt32=0
     MaximumSpawnGroups:NiagaraInt32=2
     Vector Group 0:Vector3f=<1 1 1>
     Vector Group 1:Vector3f=&VectorFromFloat(Value=20)
     Vector Group 2:Vector3f=<0 0 0>
     Vector Group 3:Vector3f=<0 0 0>
     Vector Group 4:Vector3f=<0 0 0>
     Vector Group 5:Vector3f=<0 0 0>
    )
    ScaleA=true
    ScaleRGB=true
   ) {module=ScaleColor001}
   &ScaleColor(Scale Alpha=2, ScaleA=true, ScaleRGB=true) {enabled=false}
   &DynamicMaterialParameters(
    Index 0 Param 2=&FloatFromCurve(FloatCurve=DataInterface() {propertyValues {Curve {keys [<at 0 15>,<at 1 8>]}}})
    Index 0 Param 4=&FloatFromCurve(FloatCurve=DataInterface() {propertyValues {Curve {keys [<at 0 0>, {arriveTangent=0.1, interpMode=RCIM_Cubic, leaveTangent=0.1, time=0.1, value=15}, <at 1 0.1>]}}})
    Module.Index0Float0Write=true
    Module.Index0Float1Write=true
    Module.Index0Float2Write=true
    Module.Index0Float3Write=true
   )
   /SetVariables(Particles.R1 ColorScale:LinearColor=&VC_DynamicScaleColor(
     Color to Scale:LinearColor=$Particles.R1 Color
     RGB Scale:Vector3f=&VectorFromFloat(Value=20)
     Alpha Scale:NiagaraFloat=&FloatFromCurve(FloatCurve=DataInterface() {propertyValues {Curve {keys [<at 0 1>,<at 1 0>]}}})
    )) {enabled=false, isSetParameters=true, module=SetVariables_#3}
  ]
  rendererClasses [&NiagaraSpriteRendererProperties]
  renderers [
   {rendererClass=&NiagaraSpriteRendererProperties}
   {rendererClass=&NiagaraSpriteRendererProperties, rendererIndex=1}
  ]
 }
 RIBBON {
  emitterUpdateScript [
   &EmitterState()
   &VC_FilteredSocketCount()
   &VC_RateQualityHandler(High Spawn Rate=50, Low Spawn Rate=30, Medium Spawn Rate=45, Use Custom Rate?=true)
   &VC_SpawnRibbon_v1(SpawnRate=$Emitter.Spawn Rate)
  ]
  particleSpawnScript [
   &VC_PaintedVariantsOverride(
    Black=<rgba 0 0 0 2>
    BurntSienna=<rgba 0.4522 0.1948 0.1402 1.2>
    Cobalt=<rgba 0 0.125 1.25 1>
    Crimson=<rgba 0.5729 0.0205 0.0014 1>
    ForestGreen=<rgba 0.0369 0.5365 0.0012 1.5>
    Grey=<rgba 0.4427 0.4427 0.4427 0.8>
    Lime=<rgba 0.825 1.25 0.0007 1>
    Orange=<rgba 1.3 0.2046 0 1>
    Pink=<rgba 1 0.0469 0.6534 1>
    Purple=<rgba 0.1734 0 1.7 1.5>
    Saffron=<rgba 0.8667 0.6644 0 1.5>
    SkyBlue=<rgba 0 0.7985 1.2 1>
    TitaniumWhite=<rgba 1 1 1 1>
   )
   &VC_InitializeRibbon_v1(
    Color=$Particles.Color
    Custom Offset=true
    Lifetime=&MakeCustomFloatFromBool(Bool=$User.bIsFrontendPreview, True Float=0.5, False Float=1)
    Local Vector Offset=&AddVectorToPosition(
     Position=<-70 0 0>
     Vector=&AddVectorToPosition(Position=$User.DTCarOffset, Vector=Linked() {linkedVariable {name=User.SingleBoostPos, type=NiagaraPosition}})
    )
    Offset Amount=1
    Ribbon Width=150
   )
   &AddVelocity(Velocity=&MakeCustomVectorFromBool(Bool=$User.bIsFrontendPreview, True Vector=<-2500 0 0>, False Vector=<-1000 0 0>)) {module=AddVelocity001}
   &Color(Color=&MakeCustomLinearColorFromBool(Bool=$User.bIsPainted, True Color=$User.PaintedColor, False Color=<rgba 1 1 1 1>)) {enabled=false}
  ]
  particleUpdateScript [
   &ParticleState()
   &InheritSourceMovement(
    Applied Position Scale=&VectorFromFloat(Value=&FloatFromCurve(FloatCurve=DataInterface() {propertyValues {Curve {keys [
          {interpMode=RCIM_Cubic, tangentMode=RCTM_Break, time=0, value=1}
          {arriveTangent=-2.9025, interpMode=RCIM_Cubic, tangentMode=RCTM_Break, time=1, value=0}
         ]}}}))
    ApplyForce=false
    ApplyPosition=true
    ApplyPositionFalloff=false
   )
   &SolveForcesAndVelocity(Clamp Velocity=false, Limit Acceleration=false)
   &VC_UpdateRibbon_v1()
   &DynamicMaterialParameters(
    Index 0 Param 1=&MakeCustomFloatFromBool(Bool=$User.bIsPainted, True Float=0, False Float=1)
    Index 0 Param 2=2
    Index 0 Param 3=1.3
    Index 0 Param 4=0.3
    Module.Index0Float0Write=true
    Module.Index0Float1Write=true
    Module.Index0Float2Write=true
    Module.Index0Float3Write=true
   )
   &ScaleRibbonWidth(Ribbon Width Scale=&FloatFromCurve(FloatCurve=DataInterface() {propertyValues {Curve {keys [<at 0 0.8>,<at 1 0.3>]}}}))
   &ScaleColor(
    Scale Alpha=8
    Scale RGB=&MakeCustomVectorFromBool(Bool=$User.bIsPainted, True Vector=<1 1 1>, False Vector=<0.5 0.5 0.5>)
    ScaleA=true
    ScaleRGB=true
   )
  ]
  rendererClasses [&NiagaraRibbonRendererProperties]
  renderers [{rendererClass=&NiagaraRibbonRendererProperties}]
 }
 dots {
  emitterUpdateScript [
   &EmitterState()
   &SpawnRate(
    SpawnRate=&MakeCustomFloatFromBool(Bool=$User.bIsFrontendPreview, True Float=12, False Float=24)
    Use Spawn Probability=false
   )
  ]
  particleSpawnScript [
   &VC_PaintedVariantsOverride(
    Black=&SelectLinearColorFromArray(
     Array Sampling Mode=::NewEnumerator0
     Direct Array Index=?
     Color Selection Array=DataInterface() {propertyValues {ColorData [<rgba 0.0156 0.0156 0.0156 0.5>, <rgba 0.0208 0.0208 0.0208 0.4>]}}
     Random Seed=0
    )
    BurntSienna=&SelectLinearColorFromArray(
     Array Sampling Mode=::NewEnumerator0
     Direct Array Index=?
     Color Selection Array=DataInterface() {propertyValues {ColorData [<rgba 0.3177 0.0688 0 1>, <rgba 0.1198 0.0322 0.0087 1>]}}
     Random Seed=0
    )
    Cobalt=&SelectLinearColorFromArray(
     Array Sampling Mode=::NewEnumerator0
     Direct Array Index=?
     Color Selection Array=DataInterface() {propertyValues {ColorData [<rgba 0 0.0463 2.5 1>, <rgba 0 0.0095 0.5117 1>]}}
     Random Seed=0
    )
    Crimson=&SelectLinearColorFromArray(
     Array Sampling Mode=::NewEnumerator0
     Direct Array Index=?
     Color Selection Array=DataInterface() {propertyValues {ColorData [<rgba 2.25 0 0.061 1>, <rgba 0.3438 0 0.0093 1>]}}
     Random Seed=0
    )
    ForestGreen=&SelectLinearColorFromArray(
     Array Sampling Mode=::NewEnumerator0
     Direct Array Index=?
     Color Selection Array=DataInterface() {propertyValues {ColorData [<rgba 0 0.944 0.0477 1>, <rgba 0 0.3698 0.0187 1.2>]}}
     Random Seed=0
    )
    Grey=&SelectLinearColorFromArray(
     Array Sampling Mode=::NewEnumerator0
     Direct Array Index=?
     Color Selection Array=DataInterface() {propertyValues {ColorData [<rgba 0.2552 0.2552 0.2552 1>, <rgba 0.05 0.05 0.05 1>]}}
     Random Seed=0
    )
    Lime=&SelectLinearColorFromArray(
     Array Sampling Mode=::NewEnumerator0
     Direct Array Index=?
     Color Selection Array=DataInterface() {propertyValues {ColorData [<rgba 0.5 1 0 1>, <rgba 0.2422 0.4844 0 1>]}}
     Random Seed=0
    )
    Orange=&SelectLinearColorFromArray(
     Array Sampling Mode=::NewEnumerator0
     Direct Array Index=?
     Color Selection Array=DataInterface() {propertyValues {ColorData [<rgba 2 0.5915 0.0021 1>, <rgba 0.1198 0.0312 0 1>]}}
     Random Seed=0
    )
    Pink=&SelectLinearColorFromArray(
     Array Sampling Mode=::NewEnumerator0
     Direct Array Index=?
     Color Selection Array=DataInterface() {propertyValues {ColorData [<rgba 2 0.0104 1.2765 1>, <rgba 0.3542 0 0.2254 1>]}}
     Random Seed=0
    )
    Purple=&SelectLinearColorFromArray(
     Array Sampling Mode=::NewEnumerator0
     Direct Array Index=?
     Color Selection Array=DataInterface() {propertyValues {ColorData [<rgba 0.145 0 0.8698 1>, <rgba 0.0521 0 0.3125 1>]}}
     Random Seed=0
    )
    Saffron=&SelectLinearColorFromArray(
     Array Sampling Mode=::NewEnumerator0
     Direct Array Index=?
     Color Selection Array=DataInterface() {propertyValues {ColorData [<rgba 2 1.4416 0 1>, <rgba 0.4167 0.3074 0 2>]}}
     Random Seed=0
    )
    SkyBlue=&SelectLinearColorFromArray(
     Array Sampling Mode=::NewEnumerator0
     Direct Array Index=?
     Color Selection Array=DataInterface() {propertyValues {ColorData [<rgba 0 0.1871 0.2812 1>, <rgba 0 1.3308 2 1>]}}
     Random Seed=0
    )
    TitaniumWhite=&SelectLinearColorFromArray(
     Array Sampling Mode=::NewEnumerator0
     Direct Array Index=?
     Color Selection Array=DataInterface() {propertyValues {ColorData [<rgba 0.4688 0.4688 0.4688 1>, <rgba 2 2 2 1>]}}
     Random Seed=0
    )
   )
   &InitializeParticle(
    Color Mode=::NewEnumerator0
    Lifetime=&RandomRangeFloat(Minimum=0.6, Maximum=0.5)
    Position Offset=&AddVectorToPosition(
     Position=<-70 0 0>
     Vector=&AddVectorToPosition(Position=$User.DTCarOffset, Vector=Linked() {linkedVariable {name=User.SingleBoostPos, type=NiagaraPosition}})
    )
    Position Offset Coordinate Space=::Local
    Sprite Rotation Mode=::NewEnumerator2
    Sprite Size Mode=::NewEnumerator1
    Uniform Sprite Size Max=6
    Uniform Sprite Size Min=2.5
    UsePositionOffset=true
    Write Material Random=true
   )
   &AddVelocity(Velocity=&MakeCustomVectorFromBool(
     Bool=$User.bIsFrontendPreview
     True Vector=&RandomRangeVector(Minimum=<-700 0 0>, Maximum=<-1000 1 0>)
     False Vector=<-800 0 0>
    ))
   &ShapeLocation(MaskSpawnGroup=false, Override Seed=false, Sphere Radius=35, UseOwnerScale=true)
   &Color(
    Color=&MakeCustomLinearColorFromBool(
     Bool=$User.bIsPainted
     True Color=&Multiply_LinearColorByFloat(Linear Color=$Particles.Color, Float=3)
     False Color=&RandomRangeLinearColor(Minimum=<rgba 0 0 0 1>, Maximum=<rgba 1 1 1 1>, Color Channel Mode=::NewEnumerator2)
    )
    Scale Alpha=5
   ) {enabled=false}
   &Color(
    Color=&MakeCustomLinearColorFromBool(
     Bool=$User.bIsPainted
     True Color=&Multiply_LinearColorByFloat(Linear Color=$Particles.Color, Float=3)
     False Color=&RandomRangeLinearColor(Minimum=<rgba 0 0 0 1>, Maximum=<rgba 1 1 1 1>, Color Channel Mode=::NewEnumerator2)
    )
    Scale Alpha=5
    Scale Color=&MakeCustomVectorFromBool(
     Bool=$User.bIsPainted
     True Vector=<1 1 1>
     False Vector=&RandomRangeVector(Minimum=<1 1 1>, Maximum=<8 8 8>)
    )
   ) {module=Color001}
  ]
  particleUpdateScript [
   &ParticleState()
   &ScaleSpriteSize(Scale Sprite Size Mode=::NewEnumerator0, Uniform Scale Factor=2)
   &InheritSourceMovement(
    Applied Position Scale=&VectorFromFloat(Value=&FloatFromCurve(FloatCurve=DataInterface() {propertyValues {Curve {keys [
          {interpMode=RCIM_Cubic, tangentMode=RCTM_Break, time=0, value=1}
          {arriveTangent=-2.9025, interpMode=RCIM_Cubic, tangentMode=RCTM_Break, time=1, value=0}
         ]}}}))
    ApplyForce=false
    ApplyPosition=true
    ApplyPositionFalloff=false
   )
   &SolveForcesAndVelocity(Clamp Velocity=false, Limit Acceleration=false)
   &ScaleColor(
    Scale Alpha=&MakeCustomFloatFromBool(
     Bool=$User.bIsFrontendPreview
     True Float=&FloatFromCurve(FloatCurve=DataInterface() {propertyValues {Curve {keys [<at 0 0>,<at 0.2 1>,<at 1 0>]}}})
     False Float=&FloatFromCurve(FloatCurve=DataInterface() {propertyValues {Curve {keys [<at 0 0>,<at 0.2 1>,<at 0.7 0>]}}})
    )
    Scale RGB=&MakeCustomVectorFromBool(
     Bool=$User.bIsPainted
     True Vector=&VectorFromFloat(Value=1.5)
     False Vector=&VectorFromFloat(Value=&RandomRangeFloat(
       Minimum=&SelectFloatFromArray(
        Array Sampling Mode=::NewEnumerator0
        Direct Array Index=?
        Float Selection Array=DataInterface() {propertyValues {FloatData [0.1,0.2]}}
        Random Seed=0
       )
       Maximum=5
      ))
    )
    ScaleA=true
    ScaleRGB=true
   )
  ]
  rendererClasses [&NiagaraSpriteRendererProperties]
  renderers [{rendererClass=&NiagaraSpriteRendererProperties}]
 }
 rays {
  emitterUpdateScript [
   &EmitterState(Life Cycle Mode=::NewEnumerator1, Loop Duration=2, UseLoopDelay=false)
   &SpawnRate(SpawnGroup=1, SpawnRate=300, Use Spawn Probability=false) {module=SpawnRate001}
   &SpawnRate(SpawnRate=20, Use Spawn Probability=false)
   &VC_FilteredSocketCount()
  ]
  particleSpawnScript [
   &VC_PaintedVariantsOverride(
    Black=<rgba 0.3 0.3 0.3 2>
    BurntSienna=<rgba 3 1.2 0.3 2>
    Cobalt=<rgba 0 0.1667 10 1>
    Crimson=<rgba 4 0.3333 0 1>
    DefaultColor=&SelectLinearColorFromArray(
     Array Sampling Mode=::NewEnumerator0
     Direct Array Index=?
     Direct Array Mode=::NewEnumerator1
     Color Selection Array=DataInterface() {propertyValues {ColorData [<rgba 1 0.5 0.1 3>, <rgba 0.1 1 0.1 3>, <rgba 0.1 0.1 1 3>, <rgba 1 0.1 0.1 3>, <rgba 0 0 0 0>]}}
     Random Seed=0
    )
    ForestGreen=<rgba 0 1 0.0505 1>
    Grey=<rgba 0.7 0.7 0.7 1>
    Lime=<rgba 0.75 1.5 0 1>
    Orange=<rgba 2 0.5 0 1>
    Pink=<rgba 2 0.2396 1.3598 1>
    Purple=<rgba 0.1667 0 2 1>
    R1 BurntSienna:LinearColor=<rgba 1 0.2904 0.1 1>
    R1 Cobalt:LinearColor=<rgba 0 0.0833 5 1>
    R1 Crimson:LinearColor=<rgba 3 0 0.1 1>
    R1 ForestGreen:LinearColor=<rgba 0 2 0.1011 1>
    R1 Grey:LinearColor=<rgba 0.4 0.4 0.4 1>
    R1 Lime:LinearColor=<rgba 1.1 2 0.2 1>
    R1 Orange:LinearColor=<rgba 1 0.3488 0 1>
    R1 Pink:LinearColor=<rgba 2 0.2396 1.3598 1>
    R1 Purple:LinearColor=<rgba 0.1667 0 2 1>
    R1 Saffron:LinearColor=<rgba 2 1.58 0.2 1>
    R1 SkyBlue:LinearColor=<rgba 0 1.3308 2 1>
    R1 TitaniumWhite:LinearColor=<rgba 2 2 2 1>
    Saffron=<rgba 1 0.7667 0 1.5>
    SkyBlue=<rgba 0.5 3.4942 5 0.8>
    TitaniumWhite=<rgba 2 2 2 1>
    Use R1 Painted Overrides=true
   ) {enabled=false}
   /SetVariables(Particles.vistag:NiagaraInt32=&SpawnGroup()) {isSetParameters=true, module=SetVariables_#2}
   &InitializeParticle(
    Color Mode=::NewEnumerator0
    Lifetime=&VC_FloatBySpawnGroup(
     SpawnGroup or ExecIndex:NiagaraInt32=0
     CustomInt:NiagaraInt32=0
     MaximumSpawnGroups:NiagaraInt32=2
     Float SG 0:NiagaraFloat=1
     Float SG 1:NiagaraFloat=&RandomRangeFloat(Minimum=0.1, Maximum=0.3)
     Float SG 2:NiagaraFloat=0
     Float SG 3:NiagaraFloat=0
     Float SG 4:NiagaraFloat=0
     Float SG 5:NiagaraFloat=0
    )
    Position Offset=&AddVectorToPosition(
     Position=<-70 0 0>
     Vector=&AddVectorToPosition(Position=$User.DTCarOffset, Vector=Linked() {linkedVariable {name=User.SingleBoostPos, type=NiagaraPosition}})
    )
    Position Offset Coordinate Space=::Local
    Sprite Rotation Mode=::NewEnumerator2
    Sprite Size Mode=::NewEnumerator3
    Uniform Sprite Size=&VC_FloatBySpawnGroup(
     SpawnGroup or ExecIndex:NiagaraInt32=0
     CustomInt:NiagaraInt32=0
     MaximumSpawnGroups:NiagaraInt32=2
     Float SG 0:NiagaraFloat=11.6
     Float SG 1:NiagaraFloat=&Vector2DLength(Vector:Vector2f=<15 25>)
     Float SG 2:NiagaraFloat=0
     Float SG 3:NiagaraFloat=0
     Float SG 4:NiagaraFloat=0
     Float SG 5:NiagaraFloat=0
    )
    UsePositionOffset=true
    Write Material Random=true
   )
   &ShapeLocation(MaskSpawnGroup=true, Override Seed=false, Spawn Group Mask=1, Sphere Radius=40, UseOwnerScale=true)
   &AddVelocity(Velocity=<-1500 0 0>) {enabled=false}
   &Color(Color=&VC_Dynamic_ColorBySpawnGroup(
     Color SG 0:LinearColor=&MakeCustomLinearColorFromBool(
      Bool=$User.bIsPainted
      True Color=&Multiply_LinearColorByFloat(Linear Color=$User.PaintedColor_2, Float=2)
      False Color=&SelectLinearColorFromArray(
       Array Sampling Mode=::NewEnumerator0
       Direct Array Index=?
       Direct Array Mode=::NewEnumerator1
       Color Selection Array=DataInterface() {propertyValues {ColorData [<rgba 1 0.5 0.1 5>, <rgba 0.1 1 0.1 3>, <rgba 0.1 0.1 1 5>, <rgba 1 0.1 0.1 3>, <rgba 0.1 0.1 1 5>]}}
       Random Seed=0
      )
     )
     MaximumSpawnGroups:NiagaraInt32=2
     Color SG 1:LinearColor=&MakeCustomLinearColorFromBool(Bool=$User.bIsPainted, True Color=$User.PaintedColor_3, False Color=<rgba 1 1 1 1>)
     Color SG 2:LinearColor=<rgba 1 1 1 1>
     Color SG 3:LinearColor=<rgba 1 1 1 1>
     Color SG 4:LinearColor=<rgba 1 1 1 1>
    ))
  ]
  particleUpdateScript [
   &ParticleState()
   &ScaleSpriteSize(
    Scale Sprite Size Mode=::NewEnumerator0
    Uniform Scale Factor=&VC_FloatBySpawnGroup(
     SpawnGroup or ExecIndex:NiagaraInt32=0
     CustomInt:NiagaraInt32=0
     MaximumSpawnGroups:NiagaraInt32=2
     Float SG 0:NiagaraFloat=20
     Float SG 1:NiagaraFloat=&FloatFromCurve(FloatCurve=DataInterface() {propertyValues {Curve {keys [{tangentMode=RCTM_User, time=0, value=0}, {tangentMode=RCTM_User, time=0.5, value=1}]}}})
     Float SG 2:NiagaraFloat=0
     Float SG 3:NiagaraFloat=0
     Float SG 4:NiagaraFloat=0
     Float SG 5:NiagaraFloat=0
    )
   )
   &SolveForcesAndVelocity(Clamp Velocity=false, Limit Acceleration=false)
   &ScaleColor(
    Scale Alpha=&VC_FloatBySpawnGroup(
     SpawnGroup or ExecIndex:NiagaraInt32=0
     CustomInt:NiagaraInt32=0
     MaximumSpawnGroups:NiagaraInt32=2
     Float SG 0:NiagaraFloat=2
     Float SG 1:NiagaraFloat=&FloatFromCurve(FloatCurve=DataInterface() {propertyValues {Curve {keys [<at 0 1>,<at 1 0>]}}})
     Float SG 2:NiagaraFloat=0
     Float SG 3:NiagaraFloat=0
     Float SG 4:NiagaraFloat=0
     Float SG 5:NiagaraFloat=0
    )
    Scale RGB=&VC_VectorBySpawnGroup(
     SpawnGroup or ExecIndex:NiagaraInt32=0
     CustomInt:NiagaraInt32=0
     MaximumSpawnGroups:NiagaraInt32=2
     Vector Group 0:Vector3f=<1 1 1>
     Vector Group 1:Vector3f=&VectorFromFloat(Value=20)
     Vector Group 2:Vector3f=<0 0 0>
     Vector Group 3:Vector3f=<0 0 0>
     Vector Group 4:Vector3f=<0 0 0>
     Vector Group 5:Vector3f=<0 0 0>
    )
    ScaleA=true
    ScaleRGB=true
   ) {module=ScaleColor001}
   &ScaleColor(Scale Alpha=2, ScaleA=true, ScaleRGB=true) {enabled=false}
   &DynamicMaterialParameters(
    Index 0 Param 2=&FloatFromCurve(FloatCurve=DataInterface() {propertyValues {Curve {keys [<at 0 15>,<at 1 8>]}}})
    Index 0 Param 4=&FloatFromCurve(FloatCurve=DataInterface() {propertyValues {Curve {keys [<at 0 0>, {arriveTangent=0.1, interpMode=RCIM_Cubic, leaveTangent=0.1, time=0.1, value=15}, <at 1 0.1>]}}})
    Module.Index0Float0Write=true
    Module.Index0Float1Write=true
    Module.Index0Float2Write=true
    Module.Index0Float3Write=true
   )
   /SetVariables(Particles.R1 ColorScale:LinearColor=&VC_DynamicScaleColor(
     Color to Scale:LinearColor=$Particles.R1 Color
     RGB Scale:Vector3f=&VectorFromFloat(Value=20)
     Alpha Scale:NiagaraFloat=&FloatFromCurve(FloatCurve=DataInterface() {propertyValues {Curve {keys [<at 0 1>,<at 1 0>]}}})
    )) {enabled=false, isSetParameters=true, module=SetVariables_#3}
  ]
  rendererClasses [&NiagaraSpriteRendererProperties]
  renderers [
   {rendererClass=&NiagaraSpriteRendererProperties}
   {rendererClass=&NiagaraSpriteRendererProperties, rendererIndex=1}
  ]
 }
}
summary {
 systemName=NS_Boost_Disco
 emitters [
  {emitterName=BOLADISCO, rendererClasses [&NiagaraSpriteRendererProperties, &NiagaraMeshRendererProperties]}
  {emitterName=RIBBON, rendererClasses [&NiagaraRibbonRendererProperties]}
  {emitterName=dots, rendererClasses [&NiagaraSpriteRendererProperties]}
  {emitterName=rays, rendererClasses [&NiagaraSpriteRendererProperties]}
  {emitterName=Exp, rendererClasses [&NiagaraSpriteRendererProperties]}
 ]
 userVariables [
  User.bIsFrontendPreview {type=NiagaraBool, defaultValue=true}
  User.bIsPainted {type=NiagaraBool}
  User.ColorVariantID {type=NiagaraInt32}
  User.DTCarOffset {type=NiagaraPosition, description="", defaultValue {}}
  User.Fresnel {type=NiagaraFloat, defaultValue=7, description=""}
  User.GameModeID {type=NiagaraInt32}
  User.Gap_Metallic/Spec {type=NiagaraFloat, defaultValue=0, description=""}
  User.Gap_Roughness {type=NiagaraFloat, defaultValue=0.75, description=""}
  User.PaintedColor {type=LinearColor, description="", defaultValue {struct=LinearColor, value=<rgba 1 1 1 1>}}
  User.PaintedColor_2 {type=LinearColor, description="", defaultValue {struct=LinearColor, value=<rgba 1 1 1 1>}}
  User.PaintedColor_3 {type=LinearColor, description="", defaultValue {struct=LinearColor, value=<rgba 1 1 1 1>}}
  User.PaintedScalar {type=NiagaraFloat, defaultValue=1, description=""}
  User.PortCount {type=NiagaraInt32, defaultValue=1}
  User.SingleBoostPos {type=NiagaraPosition, description="", defaultValue {}}
  User.Square_Metallic/Specular {type=NiagaraFloat, defaultValue=1, description=""}
  User.Square_Roughness {type=NiagaraFloat, defaultValue=0.35, description=""}
 ]
}
