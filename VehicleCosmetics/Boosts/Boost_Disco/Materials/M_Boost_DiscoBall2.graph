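# Lit material for the disco ball in the Disco vehicle boost effect, usable on Niagara particles and geometry caches.
# Tiled normal and Diffuse textures supply surface detail and mask, with named reroutes sharing these and particle color.
# Emissive combines a rainbow Fresnel, a rim Fresnel and glints from noise sampled with distorted tangent-space reflection vectors.
# Dynamic parameters tune glow, metallic and roughness, though both surface lerps read only the Gap values, leaving Square unused.
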Material /VehicleCosmetics/Boosts/Boost_Disco/Materials/M_Boost_DiscoBall2
graph {
 properties {
  bAutomaticallySetUsageInEditor=false
  bUsedWithGeometryCache=true
  bUsedWithNiagaraMeshParticles=true
  bUsedWithNiagaraRibbons=true
  bUsedWithNiagaraSprites=true
  bUsedWithStaticLighting=true
  refractionMethod=RM_PixelNormalOffset
  usedShadingModels=Default Lit
 }
 nodes [
  n1 = MaterialFunctionCall(Invert Fresnel=StaticBool(), Power=Constant() {r=4}) {materialFunction=&Fresnel_Function}
  n2 = Multiply(A=TextureSample(UVs=TextureCoordinate() {uTiling=3}) {texture=/VehicleCosmetics/Boosts/Boost_Disco/Textures/T_VCVFX_Boost_Disco_Rainbow_Fresnel_Diffuse}.RGB) {constB=5}
  n3 = DynamicParameter() {defaultValue=<rgba 10 30 1 1>, paramNames [Fresnel brigthness, Noise Brigthness, NoiseAddBrightness, rainbow]}
  n4 = Add(
   A=Transform(None=ReflectionVectorWS()) {transformSourceType=TRANSFORMSOURCE_World, transformType=TRANSFORM_Tangent}
   B=Multiply(A=NamedRerouteUsage()) {constB=0.8}
  )
  n5 = TextureSample() {texture=/DealershipSharedLibrary/FX/Textures/Noise/T_VCVFX_BillowNoiseTiledOffset01_RGB}
  n6 = Multiply(A=Add(A=ComponentMask(None=@n4) {g=true, r=true}, B=ComponentMask(None=@n5.RGB) {g=true, r=true}), B=Constant() {r=0.04})
  n7 = TextureCoordinate() {uTiling=2}
  MaterialExpressionConstantClamp_13 = Saturate(None=Multiply(
    A=TextureSample(UVs=@n6) {texture=/DealershipSharedLibrary/FX/Textures/Noise/T_VCVFX_NoiseGeneric01Pack_RGB}.B
    B=TextureSample(UVs=@n7) {texture=&T_VCVFX_NoiseGeneric01Pack_RGB}.G
   ))
  n8 = Multiply(A=@MaterialExpressionConstantClamp_13, B=Constant() {r=50000})
  n9 = Add(
   A=Multiply(A=Multiply(A=@n1.Result, B=@n2), B=@n3.rainbow)
   B=Power(Base=Multiply(A=@n8, B=@n3.Noise Brigthness) {constB=30}, Exp=@n3.NoiseAddBrightness)
  )
  MaterialExpressionConstantClamp_12 = Saturate(None=@n9)
  MaterialFunctionCall(Emissive=NamedRerouteUsage()) {materialFunction=&MF_FX_EmissiveAdjustments}
  n10 = TextureCoordinate() {uTiling=6, vTiling=3}
  NamedRerouteDeclaration(None=TextureSample(UVs=@n10) {samplerType=SAMPLERTYPE_Normal, texture=/VehicleCosmetics/Boosts/Boost_Disco/Textures/T_VCVFX_Boost_Disco_BallNM_N}.RGB) {name=Normal, nodeColor=<rgba 0 0 1 1>}
  NamedRerouteDeclaration(None=TextureSampleParameter2D(UVs=@n10) {parameterName=Diffuse, texture=/VehicleCosmetics/Boosts/Boost_Disco/Textures/T_VCVFX_Boost_Disco_Ball_E_Diffuse}.RGB) {name=RGB Mask, nodeColor=<rgba 0 1 0.2734 1>}
  MaterialExpressionMeshEmitterVertexColor_0 = ParticleColor()
  NamedRerouteDeclaration(None=@MaterialExpressionMeshEmitterVertexColor_0.RGB) {name=Particle Color, nodeColor=<rgba 1 0 0.0469 1>}
  NamedRerouteDeclaration(None=LinearInterpolate(B=NamedRerouteUsage(), Alpha=NamedRerouteUsage()) {constA=-1}) {name=Base Color, nodeColor=<rgba 0 0.3672 1 1>}
  n11 = NamedRerouteUsage()
  n12 = MaterialFunctionCall(Power=Constant() {r=7}) {materialFunction=&Fresnel_Function}
  n13 = Multiply(
   A=Add(A=LinearInterpolate(B=@n9, Alpha=@n11) {constA=-1, constB=0.35}, B=Multiply(A=@n12.Result, B=@n3.Fresnel brigthness) {constB=10})
   B=NamedRerouteUsage()
  )
  NamedRerouteDeclaration(None=@n13) {name=Emmisive, nodeColor=<rgba 0.6875 1 0 1>}
  n14 = DynamicParameter() {parameterIndex=1, defaultValue=<rgba 0 1 0.75 0.35>, paramNames [Gap-Metallic/Spec, Square-Metallic/Spec, Gap-Roughness, Square-Roughness]}
  NamedRerouteDeclaration(None=LinearInterpolate(A=@n14.Gap-Metallic/Spec, B=@n14.Gap-Metallic/Spec, Alpha=NamedRerouteUsage())) {name=Metallic/Spec, nodeColor=<rgba 0.9922 0 1 1>}
  NamedRerouteDeclaration(None=LinearInterpolate(A=@n14.Gap-Roughness, B=@n14.Gap-Roughness, Alpha=NamedRerouteUsage()) {constA=0.75, constB=0.35}) {name=Roughness, nodeColor=<rgba 1 0.4562 0.105 1>}
  NamedRerouteUsage()
  NamedRerouteUsage()
  NamedRerouteUsage()
  NamedRerouteUsage()
  Reroute(None=@n13)
  Reroute(None=@MaterialExpressionMeshEmitterVertexColor_0.RGB)
 ]
}
