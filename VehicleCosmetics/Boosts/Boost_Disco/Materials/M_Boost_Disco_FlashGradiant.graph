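# M_Boost_Disco_FlashGradiant is an unlit additive Niagara sprite material that draws a soft radial flash glow for the Disco vehicle boost cosmetic.
# It builds the gradient procedurally without textures, raising inverted UV distance from the sprite center to a high power for tight falloff.
# Particle color tints the gradient through MF_FX_EmissiveAdjustments.
# Particle alpha times the gradient is saturated and fed into MF_VCVFX_DepthFade as opacity.
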Material /VehicleCosmetics/Boosts/Boost_Disco/Materials/M_Boost_Disco_FlashGradiant
graph {
 properties {
  bAutomaticallySetUsageInEditor=false
  bUsedWithNiagaraSprites=true
  blendMode=BLEND_Additive
  shadingModel=MSM_Unlit
  usedShadingModels=Unlit
 }
 nodes [
  MaterialExpressionVertexColor_0 = ParticleColor()
  n1 = Add(A=TextureCoordinate(), B=Constant() {r=-0.5})
  n2 = DotProduct(A=@n1, B=@n1)
  n3 = Power(Base=OneMinus(None=SquareRoot(None=@n2)), Exp=Constant() {r=12})
  MaterialFunctionCall(Emissive=Multiply(A=@MaterialExpressionVertexColor_0.RGB, B=@n3)) {materialFunction=&MF_FX_EmissiveAdjustments}
  MaterialExpressionConstantClamp_0 = Saturate(None=Multiply(A=@MaterialExpressionVertexColor_0.A, B=@n3))
  MaterialFunctionCall(Opacity In=@MaterialExpressionConstantClamp_0) {materialFunction=&MF_VCVFX_DepthFade}
 ]
}
