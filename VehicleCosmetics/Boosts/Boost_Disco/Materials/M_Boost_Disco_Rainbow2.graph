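# Unlit, translucent, two-sided Niagara ribbon material for a rainbow trail in the Disco vehicle boost effect.
# Particle color tints the rainbow texture, with the Painted dynamic parameter blending from whitened grayscale to full color through an emissive-adjustment function.
# Opacity combines a panning trail texture, fades along U, a mask and particle alpha.
# World position offset sums time-driven sines and cosines into a wave, pinned near the U=0 end.
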Material /VehicleCosmetics/Boosts/Boost_Disco/Materials/M_Boost_Disco_Rainbow2
graph {
 properties {
  bAutomaticallySetUsageInEditor=false
  bUsedWithNiagaraRibbons=true
  blendMode=BLEND_Translucent
  shadingModel=MSM_Unlit
  twoSided=true
  usedShadingModels=Unlit
 }
 nodes [
  MaterialExpressionVertexColor_1 = ParticleColor()
  n1 = TextureSample(UVs=Panner(Coordinate=TextureCoordinate()) {speedX=-1.2}) {texture=/VehicleCosmetics/Boosts/Boost_Disco/Textures/T_VCVFX_Boost_Disco_rainbow_Trailcombo_RGB}
  n2 = Constant() {r=10}
  n3 = ComponentMask(None=TextureCoordinate()) {r=true}
  n4 = ScalarParameter() {defaultValue=1, parameterName=power fade}
  n5 = ComponentMask(None=TextureCoordinate()) {r=true}
  n6 = Multiply(A=Multiply(A=Multiply(A=@n1.B, B=@n2), B=Power(Base=@n3, Exp=@n4)), B=Power(Base=OneMinus(None=@n5), Exp=Constant() {r=2}))
  n7 = TextureSample() {texture=/VehicleCosmetics/Boosts/Boost_Disco/Textures/T_VCVFX_Boost_Disco_Mask}
  MaterialExpressionConstantClamp_6 = Saturate(None=Multiply(A=@MaterialExpressionVertexColor_1.A, B=Multiply(A=@n6, B=@n7.RGB)))
  n8 = TextureSample(UVs=TextureCoordinate()) {texture=/VehicleCosmetics/Boosts/Boost_Disco/Textures/T_VCVFX_Boost_Disco_rainbow_Ribbon}
  n9 = Desaturation(None=@n8.RGB)
  n10 = ScalarParameter() {defaultValue=2, parameterName=RainbowScale}
  MaterialExpressionConstantClamp_7 = Saturate(None=Add(A=Constant3Vector() {constant=<rgba 1 1 1 0>}, B=Multiply(A=@n9, B=@n10)))
  n11 = Add(A=Constant3Vector() {constant=<rgba 0.1 0.1 0.1 0>}, B=Multiply(A=@n8.RGB, B=@n10))
  n12 = DynamicParameter() {paramNames [Painted,Param2,Param3,Param4]}
  MaterialFunctionCall(Emissive=Multiply(A=@MaterialExpressionVertexColor_1.RGB, B=LinearInterpolate(A=@MaterialExpressionConstantClamp_7, B=@n11, Alpha=@n12.Painted))) {materialFunction=&MF_FX_EmissiveAdjustments}
  n13 = TextureCoordinate()
  MaterialExpressionConstantClamp_5 = Saturate(None=ConstantBiasScale(None=ComponentMask(None=@n13) {r=true}) {bias=-0.01, scale=15})
  n14 = Multiply(A=ScalarParameter() {defaultValue=2, group=WPO, parameterName=Speed}, B=OneMinus(None=Time()))
  n15 = ComponentMask(None=Multiply(A=TextureCoordinate() {coordinateIndex=1, uTiling=3}, B=ScalarParameter() {defaultValue=1, group=WPO, parameterName=tile})) {r=true}
  n16 = Add(A=@n14, B=@n15)
  n17 = Multiply(A=@n14, B=Constant() {r=0.5})
  n18 = Add(A=Add(A=Cosine(None=@n16), B=Cosine(None=@n16) {period=0.75}), B=Cosine(None=Add(A=@n15, B=@n17)) {period=2})
  n19 = Add(A=@n15, B=@n14)
  n20 = Add(A=Add(A=Sine(None=@n19), B=Sine(None=@n19) {period=2}), B=Sine(None=@n19) {period=0.75})
  n21 = Multiply(A=AppendVector(A=AppendVector(A=@n18, B=@n20), B=@n18), B=ScalarParameter() {defaultValue=1, group=WPO, parameterName=Amount})
  n22 = ScalarParameter() {defaultValue=1, group=WPO, parameterName=Intensity}
  Multiply(A=@MaterialExpressionConstantClamp_5, B=Transform(None=Multiply(A=@n21, B=@n22)))
 ]
}
