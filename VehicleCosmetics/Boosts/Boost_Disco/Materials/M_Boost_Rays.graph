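# Translucent unlit Niagara sprite material that draws animated radial rays for the Disco vehicle boost effect.
# The UV angle around the sprite center samples a panning noise mask, forming rays, with dynamic parameters offsetting and sharpening it.
# A steep radial falloff confines the rays toward the center, then particle alpha and dynamic Opacity scale the mask.
# Particle color feeds the shared MF_FX_EmissiveAdjustments function.
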Material /VehicleCosmetics/Boosts/Boost_Disco/Materials/M_Boost_Rays
graph {
 properties {
  bAutomaticallySetUsageInEditor=false
  bUsedWithNiagaraSprites=true
  blendMode=BLEND_Translucent
  shadingModel=MSM_Unlit
  usedShadingModels=Unlit
 }
 nodes [
  MaterialExpressionVertexColor_4 = ParticleColor()
  n1 = DynamicParameter() {paramNames [speed,power,Param3,Opacity]}
  n2 = Add(A=TextureCoordinate(), B=Constant() {r=-0.5})
  n3 = SquareRoot(None=DotProduct(A=@n2, B=@n2))
  MaterialExpressionConstantClamp_11 = Saturate(None=Power(Base=OneMinus(None=@n3), Exp=Constant() {r=12}))
  n4 = MaterialFunctionCall() {materialFunction=&VectorToRadialValue}
  n5 = Constant2Vector() {r=1}
  n6 = Constant()
  n7 = Panner(Coordinate=Add(A=Multiply(A=@n4.Vector Converted to Angle, B=@n5), B=AppendVector(A=@n6, B=@n1.speed))) {speedX=0.01, speedY=0.1}
  n8 = Multiply(
   A=@MaterialExpressionConstantClamp_11
   B=Power(Base=TextureSample(UVs=@n7) {texture=/VehicleCosmetics/Boosts/Boost_Disco/Textures/T_VCVFX_Boost_Disco_NoiseClud_mask}.RGB, Exp=@n1.power)
  )
  MaterialExpressionConstantClamp_10 = Saturate(None=Multiply(A=@MaterialExpressionVertexColor_4.A, B=Multiply(A=@n1.Opacity, B=@n8)))
  MaterialFunctionCall(Emissive=@MaterialExpressionVertexColor_4.RGB) {materialFunction=&MF_FX_EmissiveAdjustments}
 ]
}
